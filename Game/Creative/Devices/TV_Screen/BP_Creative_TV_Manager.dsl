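; Blueprint manager class for the Creative TV Screen device, judging by its name and folder.
; It overrides CanBeSavedToPlayset to return false, so the manager is excluded from playset saving.
; The construction script is empty, and the file shows no other logic.
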(fn CanBeSavedToPlayset
  (return false))

(fn ConstructionScript)
