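; BP_Creative_TV_Silent is a Creative TV (Llamatron) device that shows streamed video on its screen and can offer picture-in-picture.
; It registers as a source with a shared streaming manager, found or server-spawned, supplying video ID, start time and enabled-dependent priority.
; Stream state and timed content-protection checks drive the screen material.
; PIP uses gameplay effects on minigame players, retried until they stick, and client live-stream options aggregated across all TVs.
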(event EventBeginPlay
  ;[0 1 0.04321 0.5] Initialization
  (Material|ReattachOrCreateMID (Variables|Default|GetScreenMesh) (Variables|Default|GetScreenMID) 1)
  (CallFunction|ResetVideoTime)
  (bind _returnvalue (Utilities|Time|SetTimerbyEvent (delegate Custom|RecheckContentProtection) (Variables|Default|GetfRecheckContentProtection) true))
  (Variables|Default|SetRecheckTimer _returnvalue))

(event OnAddedtoMinigame(FortMinigameProgress)
  (params (Minigame :type Fort%20Minigame%20Object%20Reference))
  ;[0 1 0.3626 0.25] When getting added to Minigame, apply PIP to all then bind to wait for more players.
  (Networking|SwitchHasAuthority
    (:Authority
      (PIP|UpdatePIPOnAll Minigame (Variables|UserOptions|GetbEnablePIP))
      (Minigame|BindEventtoPlayerAddedDelegate Minigame (delegate Custom|Player%20Added)))))

(event ChildFunctions|EventPlayModeChanged
  (params (Minigame :type Fort%20Minigame%20Object%20Reference) (bIsInPlayMode :type Boolean))
  ;[0 1 0.04321 0.5] Initialization
  (|CallParentFunction Minigame bIsInPlayMode)
  (if bIsInPlayMode
    (CallFunction|ResetVideoTime)))

(event Custom|RecheckContentProtection
  ;[0 1 0.04321 0.5] Initialization
  (Default|RecheckUCP))

(event Custom|Player%20Added
  (params (UniqueNetId :type Unique%20Net%20Id%20Structure) (bIsLocalPlayer :type Boolean))
  (bind _fort_player_state (Class|CreativeCommonDeviceFunctionLibrary|GetPlayerStateFromNetID UniqueNetId))
  ;[0 1 0.3626 0.25] When getting added to Minigame, apply PIP to all then bind to wait for more players.
  (PIP|UpdateGameplayEffectIfPIP _fort_player_state (Variables|UserOptions|GetbEnablePIP))
  (Utilities|Set|Add (Variables|PIP|GetPlayerPIPRecheckSet) _fort_player_state)
  (CallFunction|RecheckGameplayEffectApplicationAfterDelay)
  (Networking|ForceNetUpdate))

(event EventInitializeDevice
  ;[0 1 0.04321 0.5] Initialization
  (|CallParentFunction)
  (CreativeObjectRegistry|GetOrCreateObjectRegister self "(TagName=\"Creative.Register.Llamatron.All\")" true 0 (Variables|Default|GetAllTVRegister))
  (Initialization|FindorCreateManager)
  (Initialization|SetupConfigurables)
  (Initialization|UpdateScreenAndAudio)
  (PIP|UpdateGlobalPIPfromAllTVs (Variables|UserOptions|GetbEnablePIP))
  (Class|CreativeStreamingManagerComponent|CheckStreamonInit (Variables|Default|GetStreaming%20Manager) self))

(event Custom|Recheck%20Gameplay%20Effect%20Application%20after%20Delay
  ;[0 1 0.3626 0.25] When getting added to Minigame, apply PIP to all then bind to wait for more players.
  ;;[1 0.07259 0 0.25] Hack - Applying a GE on the server before the client has loaded does not work correctly and will immediately cancel. This attempts reapplication when applying a GE until it sticks.
  (Utilities|FlowControl|Delay (Variables|PIP|GetGE%20Reapplication%20Delay%20Time))
  (bind _any_remaining (Range|AttemptReapplyin-RangeGameplayEffect))
  (if _any_remaining
    (CallFunction|RecheckGameplayEffectApplicationAfterDelay)))

(event EventEndPlay
  (params (EndPlayReason :type EEndPlayReason%20Enum :default "Destroyed"))
  ;[1 0.08736 0] EndPlay : Remove this player from the manager and unbind delegates.
  (PIP|UpdateGlobalPIPfromAllTVs false)
  (bind _fortminigameprogress (Variables|Default|GetFortMinigameProgress)
    (:then
      (PIP|UpdatePIPOnAll (Variables|/Script/FortniteGame.FortMinigameLogicComponent|GetCurrentMinigame _fortminigameprogress))
      (bind _streaming_manager (Variables|Default|GetStreaming%20Manager)
        (:then
          (StreamPlayerOptions|RemoveStreamSource _streaming_manager self)
          (Initialization|BindOrUnbindStreamingManagerDelegates false)
          (Utilities|Time|ClearandInvalidateTimerbyHandle (Variables|Default|GetRecheckTimer))
          (|CallParentFunction EndPlayReason))
        (:else)))
    (:else)))

(event Custom|Server%20Request%20Stream%20Update
  ;[0.1455 1 0.8701 0.25] Calls to Stream Manager
  (Networking|SwitchHasAuthority
    (:Authority
      (bind _streaming_manager (Variables|Default|GetStreaming%20Manager)
        (:then
          (Class|CreativeStreamingManagerComponent|RequestStreamingUpdate _streaming_manager self))
        (:else)))))

(event Interaction|EventBlueprintOnInteract
  (params (InteractingPawn :type Fort%20Pawn%20Object%20Reference) (InteractionBeingAttempted :type EInteractionBeingAttempted%20Enum :default "FirstInteraction"))
  ;[1 1 1] Device Interaction Logic
  (if (CallFunction|IsinPlayMode)
    (Class|CreativeEnabledComponent|SetEnabled (Variables|Default|GetEnabledComponent) (not (Class|CreativeEnabledComponent|IsEnabled (Variables|Default|GetEnabledComponent))))))

(event Custom|Server%20Force%20Stream%20Control
  ;[0.1455 1 0.8701 0.25] Calls to Stream Manager
  (Initialization|UpdateScreenAndAudio)
  (if (Variables|UserOptions|GetbResetOnStreaming)
    (CallFunction|ResetVideoTime)))

(event OnEnabledStateChanged(EnabledComponent)
  (params (Enabled :type Boolean))
  ;[1 1 1] Device Interaction Logic
  (Initialization|UpdateScreenAndAudio)
  (if (and Enabled (Variables|UserOptions|GetbResetOnEnabled))
    (CallFunction|ResetVideoTime))
  (CallFunction|ServerRequestStreamUpdate) ; Priority may have changed now this is disabled/enabled
)

(event Custom|Server%20Force%20Release%20Control
  ;[0.1455 1 0.8701 0.25] Calls to Stream Manager
  (Networking|SwitchHasAuthority
    (:Authority
      (bind _streaming_manager (Variables|Default|GetStreaming%20Manager)
        (:then
          (if (Utilities|IsCurrentStream)
            (Class|CreativeStreamingManagerComponent|ServerRequestReleaseStreamControl _streaming_manager)))
        (:else)))))

(event OnGameplayMessageReceived(EnableWhenReceivingFrom)
  (params (TriggerInstigator :type Controller%20Object%20Reference))
  ;[0.4168 0 1] Message Receivers
  (Class|CreativeEnabledComponent|SetEnabled (Variables|Default|GetEnabledComponent) true))

(event OnGameplayMessageReceived(DisableWhenReceivingFrom)
  (params (TriggerInstigator :type Controller%20Object%20Reference))
  ;[0.4168 0 1] Message Receivers
  (Class|CreativeEnabledComponent|SetEnabled (Variables|Default|GetEnabledComponent) false))

(event EventOnMinigameEnded
  ;[0.4168 0 1] Message Receivers
  (|CallParentFunction)
  (CallFunction|ServerForceReleaseControl))

(fn AttemptReapplyIn-RangeGameplayEffect
  ;[1 1 1] Checks all nearby pawns to ensure they have the InRange tag applied to them. If they don't, reapply the Gameplay Effect.
  (if (Variables|UserOptions|GetbEnablePIP)
    (bind _result (Utilities|Set|ToArray (Variables|PIP|GetPlayerPIPRecheckSet)))
    (bind (_arrayindex _arrayelement) (Utilities|Array|ReverseforEachLoop _result)
      (:LoopBody
        (Utilities|IsValid _arrayelement
          (:Is%20Valid
            (bind _returnvalue (GameplayTags|HasMatchingGameplayTag(Message) _arrayelement "(TagName=\"Granted.Creative.CreativeEffect.TV.PiPUpdateCheck\")"))
            (if (not _returnvalue)
              (PIP|UpdateGameplayEffectIfPIP _arrayelement true)
              (else
                (Utilities|Set|Remove (Variables|PIP|GetPlayerPIPRecheckSet) _arrayelement))))))
      (:Completed
        (return (Math|Conversions|ToBoolean(Integer) (Utilities|Set|Length (Variables|PIP|GetPlayerPIPRecheckSet))))))))

(fn BindorUnbindStreamingManagerDelegates
  (params (Bind :type Boolean))
  (bind _outputdelegate (EventDispatchers|CreateEvent :!event Stream%20State%20Updated))
  (bind _outputdelegate_2 (EventDispatchers|CreateEvent :!event On%20Stream%20Source%20Changed))
  (Networking|CanExecuteCosmeticEvents
    (:True
      (bind _streaming_manager (Variables|Default|GetStreaming%20Manager)
        (:then
          (if Bind
            (Default|BindEventtoOnStreamStateChanged _streaming_manager _outputdelegate)
            (Default|BindEventtoOnStreamSourceChanged _streaming_manager _outputdelegate_2)
            (else
              (Default|UnbindEventfromOnStreamStateChanged _streaming_manager _outputdelegate)
              (Default|UnbindEventfromOnStreamSourceChanged _streaming_manager _outputdelegate_2))))
        (:else)))))

(fn BlueprintCanInteract
  (params (InteractingPawn :type Fort%20Pawn%20Object%20Reference) (InteractionBeingAttempted :type EInteractionBeingAttempted%20Enum :default "FirstInteraction") (InteractionType :type TInteractionType%20Enum :default "IT_NoInteraction"))
  (return (or (and (>= (Variables|UserOptions|GetInteractTime) 0.0) (CallFunction|IsinPlayMode)) (Interaction|CanEditOptionsComponent (Variables|Default|GetToyOptionsComponent) InteractingPawn))))

(fn BlueprintGetInteractionString
  (params (InteractingPawn :type Fort%20Pawn%20Object%20Reference) (InteractionBeingAttempted :type EInteractionBeingAttempted%20Enum :default "FirstInteraction"))
  (if (CallFunction|IsinPlayMode)
    (return (select (Class|CreativeEnabledComponent|IsEnabled (Variables|Default|GetEnabledComponent)) (Variables|Default|GetDisable%20Text) (Variables|Default|GetEnable%20Text)))
    (else
      (return (Variables|/Script/FortniteGame.FortActorOptionsComponent|GetInternal_InteractionText (Variables|Default|GetToyOptionsComponent))))))

(fn BlueprintGetInteractionTime
  (params (InteractingPawn :type Fort%20Pawn%20Object%20Reference) (InteractionBeingAttempted :type EInteractionBeingAttempted%20Enum :default "FirstInteraction"))
  (return (Variables|UserOptions|GetInteractTime) (and (>= (Variables|UserOptions|GetInteractTime) 0.0) (CallFunction|IsinPlayMode))))

(fn BlueprintOnLocalInteract
  (params (InteractingPawn :type Fort%20Player%20Pawn%20Object%20Reference) (InteractionBeingAttempted :type EInteractionBeingAttempted%20Enum :default "FirstInteraction"))
  (bind _returnvalue (|CallParentFunction InteractingPawn))
  (return _returnvalue))

(fn CheckforContentProtection
  (bind _streaming_manager (Variables|Default|GetStreaming%20Manager)
    (:then
      (bind _mediasource (Variables|/Game/Creative/Devices/TV_Screen/CreativeStreamingManagerComponent.CreativeStreamingManagerComponent_C|GetMediaSource _streaming_manager)
        (:then
          (if (Utilities|Enum|Equal(Enum) (Player|ShouldProtectPlayerfromContent _mediasource) "UCPAudio")
            (return true false)
            (elif (Utilities|Enum|Equal(Enum) (Player|ShouldProtectPlayerfromContent _mediasource) "UCPVideo")
              (return false true)
              (elif (Utilities|Enum|Equal(Enum) (Player|ShouldProtectPlayerfromContent _mediasource) "UCPBoth")
                (return true true)
                (else
                  (return false false))))))
        (:else
          _)))
    (:else
      _))
  (return false false))

(fn FindOrCreateManager
  (bind _manager_register (LocalVariables|GetManager%20Register))
  ;[1 1 1] Creates a new Manager Actor if we don't already have one. This needs to be an actor rather than in a static place like the game state as we need it to deconstruct on the client when they leave, to avoid replication issues.
  ; If called on the server, this will create an object if it cannot find it. If on the client it will return null if it can't find one.
  (bind _streaming_manager (Variables|Default|GetStreaming%20Manager)
    (:then
      (return _streaming_manager))
    (:else
      (Networking|SwitchHasAuthority
        (:Authority
          (CreativeObjectRegistry|GetOrCreateObjectRegister self "(TagName=\"Creative.Register.Streaming.Manager\")" false 0 _manager_register)
          (bind _returnvalue (Class|FortCreativeIndexedObjectRegisterComponent|GetRegisteredObject _manager_register "First" 0))
          (bind _ascreative_streaming_manager_component (Utilities|Casting|CastToCreativeStreamingManagerComponent _returnvalue)
            (:then
              (bind _output_get (Variables|Default|SetStreaming%20Manager _ascreative_streaming_manager_component))
              (return _output_get))
            (:CastFailed
              (bind _returnvalue_2 (Building|SpawnBuildingActor (Variables|Default|GetStreaming%20Manager%20Class) (Transformation|GetActorTransform)))
              (bind _asbp_creative_tv_manager (Utilities|Casting|CastToBP_Creative_TV_Manager _returnvalue_2))
              (Variables|Default|SetStreaming%20Manager (Variables|/Game/Creative/Devices/TV_Screen/BP_Creative_TV_Manager.BP_Creative_TV_Manager_C|GetCreativeStreamingManagerComponent _asbp_creative_tv_manager))
              (return (Variables|Default|GetStreaming%20Manager)))))))))

(fn GetPIPValuesForScreen
  (Variables|Default|GetStreaming%20Manager
    (:then
      (return (Utilities|IsPlaying) (Variables|PIP|GetScreenNotInView)))
    (:else)))

(fn GetOverrideMeshMaterial
  (params (MeshComp :type Static%20Mesh%20Component%20Object%20Reference) (MatIdx :type Integer))
  (if (and (== MatIdx 1) (Utilities|Equal(Object) MeshComp (Variables|Default|GetScreenMesh)))
    (return (Variables|Default|GetScreenMID))
    (else
      (return (|CallParentFunction MeshComp MatIdx)))))

(fn GetStreamDetails
  (bind _videoid (VideoStreaming|Source|GetStreamingVideoID))
  (bind _stream_priority (VideoStreaming|Source|GetStreamPriority))
  (return _videoid (Variables|Default|GetStreamStartTime) _stream_priority))

(fn GetStreamingVideoID
  (return (Utilities|String|SelectString (Variables|Default|GetDefault%20Llamatron%20ID) (Variables|UserOptions|GetVideoID) (Utilities|String|IsEmpty (Variables|UserOptions|GetVideoID)))))

(fn GetStreamIsPlaying
  (return false))

(fn GetStreamPriority
  (return (Math|Float|SelectFloat (+ (Variables|Default|GetStreamPriority) (Math|Float|SelectFloat 0.0 0.1 (Utilities|String|IsEmpty (Variables|UserOptions|GetVideoID)))) -1.0 (Class|CreativeEnabledComponent|IsEnabled (Variables|Default|GetEnabledComponent)))))

(fn GetStreamStartTime
  (return (Variables|Default|GetStreamStartTime)))

(fn GetWantsToLoop
  (return (Variables|UserOptions|GetbLooping)))

(fn OnStreamSourceChanged
  (params (New%20Source :type Object%20Reference) (New%20Stream%20ID :type String))
  (Initialization|UpdateScreenAndAudio))

(fn OnRep_StreamingManager
  ;[1 1 1] If we've attempted to call Streaming Manager with it invalid, we now update as we have a valid reference.
  (Networking|CanExecuteCosmeticEvents
    (:True
      (Variables|Default|GetStreaming%20Manager
        (:then
          (Initialization|BindOrUnbindStreamingManagerDelegates true)
          (Utilities|FlowControl|Sequence
            (:then_0
              (CallFunction|UpdateStreamingState))
            (:then_1
              (Class|CreativeStreamingManagerComponent|RequestStreamingUpdate (Variables|Default|GetStreaming%20Manager) self)
              (Initialization|UpdateScreenAndAudio))))
        (:else)))))

(fn OnRep_StreamStartTime
  (bind (_stream_uid _start_time _priority) (VideoStreaming|Source|GetStreamDetails))
  (Default|CallOnTVStartTimeChanged self _start_time _priority)
  (bind _streaming_manager (Variables|Default|GetStreaming%20Manager)
    (:then
      (Class|CreativeStreamingManagerComponent|RequestStreamingUpdate _streaming_manager self))
    (:else)))

(fn RecheckUCP
  (Networking|SwitchHasAuthority
    (:Remote
      (bind _streaming_manager (Variables|Default|GetStreaming%20Manager)
        (:then
          (bind _mediasource (Variables|/Game/Creative/Devices/TV_Screen/CreativeStreamingManagerComponent.CreativeStreamingManagerComponent_C|GetMediaSource _streaming_manager)
            (:then
              (if (not (xor (Utilities|Enum|Equal(Enum) (Player|ShouldProtectPlayerfromContent _mediasource) "UCPNone") (Variables|Default|GetUCPBlocking)))
                (Initialization|UpdateScreenAndAudio)))
            (:else)))
        (:else)))))

(fn ResetVideoTime
  (Networking|SwitchHasAuthority
    (:Authority
      (Variables|Default|SetStreamStartTime (Fortnite|Sync|Time|GetServerWorldTimeSeconds))
      (CallFunction|ServerRequestStreamUpdate))))

(fn SetPIPAudio
  (params (PiP%20Status :type Boolean)))

(fn SetCachedPIPState
  (params (ShowPIP :type Boolean))
  (Variables|PIP|SetScreenNotInView ShowPIP))

(fn SetupConfigurables
  (Variables|BuildingActor|SetbIsInvulnerable (not (Variables|UserOptions|GetbDestructible)))
  (Initialization|InitializeComponent (Variables|Default|GetCreative_UseRestriction_Component) (Variables|Default|GetFortMinigameProgress) (Variables|Default|GetEnabledComponent) 0 996 -1 0 1 :!class /Game/Creative/Devices/Common/Components/Creative_UseRestriction_Component.Creative_UseRestriction_Component_C)
  (Class|CreativeEnabledComponent|InitializeComponent (Variables|Default|GetEnabledComponent) (Variables|Default|GetFortMinigameProgress) (Variables|LegacyOptions|GetEnabledIndex))
  (CallFunction|UpdateBorder)
  (bind _streaming_manager (Variables|Default|GetStreaming%20Manager)
    (:then
      (StreamPlayerOptions|AddStreamSource _streaming_manager self))
    (:else)))

(fn SkipPipCheck
  (Variables|Default|GetStreaming%20Manager
    (:then
      _)
    (:else
      (bind _streaming_manager (Initialization|FindorCreateManager)) ; May have gotten detached leaving back to hub and re-entering or\nstore mannequin) can we find it again?
      (Utilities|IsValid _streaming_manager
        (:Is%20Valid
          _)
        (:Is%20Not%20Valid
          (Variables|Initialization|SetStreaming%20Update%20Queued true) ; Can't find it, likely not created yet.\nLet the system know when the object is created that we want to stream.
          (return true)))))
  (return (not (Utilities|IsPlaying))))

(fn StreamStateUpdated
  (params (New%20Stream%20State :type ECreativeStreamPlayState%20Enum :default "NewEnumerator0"))
  (Initialization|UpdateScreenAndAudio)
  (Default|CallOnTVStateChanged New%20Stream%20State))

(fn StreamUpdated
  (params (Current%20VUID :type String) (Start%20Time :type Float%20%28double-precision%29))
  (return false))

(fn UpdateActiveTVRegisterifPIP
  (Networking|CanExecuteCosmeticEvents
    (:True
      (if (and (Utilities|IsPlaying) (Variables|UserOptions|GetbEnablePIP))
        (CreativeObjectRegistry|GetOrCreateObjectRegister self "(TagName=\"Creative.Register.LlamatronActive\")" true 0 (Variables|Default|GetActiveTVRegister))
        (else
          (bind _activetvregister (Variables|Default|GetActiveTVRegister)
            (:then
              (Class|CreativeIndexedObjectRegisterComp|RemoveObjectfromIndex _activetvregister self 0)
              (PIP|SetPIPAudio false) ; We're not active anymore, reset our PIP Audio Status
            )
            (:else)))))))

(fn UpdateGameplayEffectifPIP
  (params (PlayerState :type Player%20State%20Object%20Reference) (Enable :type Boolean))
  (bind _returnvalue (Ability|GetFortAbilitySystemComponent PlayerState))
  (bind _self self)
  (bind _returnvalue_2 (Ability|GetAbilitySystemComponent _self))
  (bind _pip_gameplay_effect (Variables|PIP|GetPIP%20Gameplay%20Effect))
  (bind _abilitycomp (LocalVariables|GetAbilityComp))
  (bind _abilitycomp_2 (LocalVariables|GetAbilityComp))
  (bind _in_range_gameplay_effect (Variables|PIP|GetIn%20Range%20Gameplay%20Effect))
  ;[0 0.7912 1 0.5] Adds or removes a stack of a Gameplay Effect to the pawn. This is used to track when they're in range of the TV Stream and will in turn be used to run PiP Checks.
  (Utilities|IsValid PlayerState
    (:Is%20Valid
      (Utilities|IsValid _returnvalue
        (:Is%20Valid
          (LocalVariables|SetAbilityComp _returnvalue)
          (Utilities|IsValid (Ability|GetAbilitySystemComponent self)
            (:Is%20Valid
              (if Enable
                ;[0 0.7912 1 0.5] Adds the Ability if it does not already exist. Removes the ability if you no longer have the tag.
                (GameplayEffects|ApplyGameplayEffectToTarget :self _returnvalue_2 :GameplayEffectClass _in_range_gameplay_effect :Target _abilitycomp_2 :Level 0.0)
                (if (not (Abilities|HasActiveAbilitywithAnyMatchingTag (LocalVariables|GetAbilityComp) "(GameplayTags=((TagName=\"Asset.Creative.CreativeAbility.TV.PIP\")))"))
                  (GameplayEffects|ApplyGameplayEffectToSelf :self _abilitycomp :GameplayEffectClass _pip_gameplay_effect :Level 0.0))
                (else
                  (GameplayEffects|RemoveActiveGameplayEffectbySourceEffect _abilitycomp_2 _in_range_gameplay_effect _returnvalue_2 -1) ; Remove All Stacks. The GE is aggregated by Source, \nso this will only remove stacks from this object.
                  (if (not (GameplayTags|HasMatchingGameplayTag (LocalVariables|GetAbilityComp) "(TagName=\"Asset.Creative.CreativeEffect.TV.InRange\")"))
                    (GameplayEffects|RemoveActiveGameplayEffectbySourceEffect _abilitycomp _pip_gameplay_effect 0 -1)))))
            (:Is%20Not%20Valid
              (Development|PrintString "CANNOT FIND ABILITY SYSTEM" true true "(R=1.000000,G=0.040959,B=0.000000,A=1.000000)" 1e+01))))))))

(fn UpdateGlobalPIPFromAllTVs
  (params (Enable%20for%20self :type Boolean))
  (bind _returnvalue (Options|GetRuntimeOptions))
  (Networking|SwitchHasAuthority
    (:Authority
      (bind _fortminigameprogress (Variables|Default|GetFortMinigameProgress)
        (:then
          (bind _currentminigame (Variables|/Script/FortniteGame.FortMinigameLogicComponent|GetCurrentMinigame _fortminigameprogress)
            (:then
              (PIP|UpdatePIPOnAll _currentminigame Enable%20for%20self))
            (:else)))
        (:else)))
    (:Remote
      (if Enable%20for%20self
        (LocalVariables|SetAnyWantsPIP true)
        (else
          ;[0 0.8698 0.7627 0.25] We don't want to show PIP, but maybe another TV does. Check every TV
          (bind _alltvregister (Variables|Default|GetAllTVRegister)
            (:then
              (bind (_bisvalid _register) (Class|FortCreativeIndexedObjectRegisterComponent|GetRegisteratIndex _alltvregister 0))
              (bind _register_objects (Utilities|Struct|BreakFortObjectArray _register))
              (if _bisvalid
                (bind _array_element (Utilities|Array|ForEachLoopwithBreak _register_objects)
                  (:LoopBody
                    (bind _asbp_creative_tv_silent (Utilities|Casting|CastToBP_Creative_TV_Silent _array_element))
                    (if (Variables|UserOptions|GetbEnablePIP _asbp_creative_tv_silent)
                      (LocalVariables|SetAnyWantsPIP true)))
                  (:Completed
                    _))
                (else
                  _))
              (if (LocalVariables|GetAnyWantsPIP)
                (Variables|/Script/FortniteGame.FortRuntimeOptions|SetbEnableLiveStream true (Options|GetRuntimeOptions)) ; Allows anything to potentially enable Live stream\n(the ability will enable or disable it)
                (else
                  (Variables|/Script/FortniteGame.FortRuntimeOptions|SetbEnableLiveStream false _returnvalue) ; Do not allow Live Stream
                  (Variables|/Script/FortniteGame.FortRuntimeOptions|SetbEnableLiveStreamInMatch false _returnvalue) ; Remove Current Live Stream
                  (Class|LiveStreamHUDElement|SetMirrorScreen -1) ; Point the live stream to an invalid material.
                )))
            (:else)))))))

(fn UpdatePIPonAll
  (params (Minigame :type Fort%20Minigame%20Object%20Reference) (Enable :type Boolean))
  (Networking|SwitchHasAuthority
    (:Authority
      (Utilities|IsValid Minigame
        (:Is%20Valid
          (for _array_element (Minigame|GetParticipatingPlayers Minigame)
            (PIP|UpdateGameplayEffectIfPIP _array_element Enable)))))))

(fn UpdateScreenandAudio
  (bind _screenmid (Variables|Default|GetScreenMID))
  (bind _devicestate (LocalVariables|GetDeviceState))
  (Networking|CanExecuteCosmeticEvents
    (:True
      (bind _streaming_manager (Variables|Default|GetStreaming%20Manager)
        (:then
          (LocalVariables|SetDeviceState (Utilities|GetCurrentState _streaming_manager))
          (bind (_blockaudio _blockvideo) (Default|CheckForContentProtection))
          (Variables|Default|SetUCPBlocking (or _blockaudio _blockvideo))
          (if _blockvideo
            (Rendering|Material|SetScalarParameterValue _screenmid "Broadcast" 0.0 :!class /Script/Engine.MaterialInstanceDynamic) ; Broadcast is image or static
            (else
              (Material|ReattachOrCreateMID (Variables|Default|GetScreenMesh) (Variables|Default|GetScreenMID) 1)
              (Rendering|Material|SetScalarParameterValue _screenmid "Broadcast" (Utilities|Select :Index _devicestate :NewEnumerator0 0.0 :NewEnumerator1 1.0 :NewEnumerator2 1.0 :NewEnumerator6 1.0 :NewEnumerator3 1.0 :NewEnumerator5 0.0) :!class /Script/Engine.MaterialInstanceDynamic) ; Broadcast is image or static
              (Rendering|Material|SetScalarParameterValue _screenmid "Enabled" (Utilities|Select :Index _devicestate :NewEnumerator0 0.0 :NewEnumerator1 0.0 :NewEnumerator2 0.0 :NewEnumerator6 0.0 :NewEnumerator3 1.0 :NewEnumerator5 1.0) :!class /Script/Engine.MaterialInstanceDynamic) ; enabled is blank screen or output of broadcast
              (Rendering|Material|SetScalarParameterValue _screenmid "Loading" (Math|Float|SelectFloat 1.0 0.0 (and (Utilities|Enum|Equal(Enum) _devicestate "NewEnumerator6") (Variables|UserOptions|GetbShowLoading))) :!class /Script/Engine.MaterialInstanceDynamic) ; enabled is blank screen or output of broadcast
            )))
        (:else)))))

(fn UpdateStreamingState
  (if (not (Building|IsPreviewBuildingActor))
    (Variables|Default|GetStreaming%20Manager
      (:then
        (Initialization|UpdateScreenAndAudio))
      (:else))))

(fn UpdateBorder
  (Networking|CanExecuteCosmeticEvents
    (:True
      (Rendering|Material|SetMaterial (Variables|Default|GetScreenMesh) 0 (select (Variables|UserOptions|GetbShowBorder) (Variables|Effects|GetFrame%20Material) (Variables|Effects|GetInvisible%20Material))))))

(fn ConstructionScript)
