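; Blueprint interface for streaming video in Creative mode devices such as the TV screen, declaring functions with only default returns.
; Getters expose the video ID, start time, playing state, loop preference, combined stream details and a priority where higher is better.
; Other functions signal that an actor is now controller, check a video ID for transmit, and pass stream updates with current ID and start time.
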(fn ActorIsNowController)

(fn CheckForTransmit
  (params (VideoID :type String)))

(fn GetStreamDetails
  (return "" 0.0 0.0))

(fn GetStreamingVideoID
  (return ""))

(fn GetStreamIsPlaying
  (return false))

(fn GetStreamPriority
  ;[1 1 1] Returns the priority of the stream. Higher is better.
  (return 0.0))

(fn GetStreamStartTime
  (return 0.0))

(fn GetWantsToLoop
  (return false))

(fn StreamUpdated
  (params (Current%20VUID :type String) (Start%20Time :type Float%20%28double-precision%29))
  (return false))
