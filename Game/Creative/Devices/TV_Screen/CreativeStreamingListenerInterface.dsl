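; Blueprint interface for listeners of Creative streaming, stored with the TV Screen device.
; It declares GetTimerLength, which returns a float, plus two callbacks with no outputs.
; PeriodicListenerUpdate passes a stream UID and start time, and TimerWithEndTimeCompleted passes an end time and stream UID.
; All three bodies are empty stubs, so timing and playback logic belongs to the implementers and callers.
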(fn GetTimerLength
  (return 0.0))

(fn PeriodicListenerUpdate
  (params (Stream%20UID :type String) (StartTime :type Float%20%28double-precision%29))
  (return))

(fn TimerWithEndTimeCompleted
  (params (EndTime :type Float%20%28double-precision%29) (Stream%20UID :type String))
  (return))
