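; Creative streaming manager component that coordinates video playback for in-island TV screen devices.
; The server picks the highest-priority stream source, unless locked, and replicates that actor, start time and play state.
; Clients in the same volume load the video URL, then open it on a shared media player with subtitles, share blocking and looping.
; Party-sync sources use a media player controller; others seek from server world time.
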(event EventBeginPlay
  (Initialization|RegisterSelfWithManager)
  ;[1 1 1] Initialization - Register with the Manager
  (Utilities|FlowControl|Sequence
    (:then_0
      (Networking|CanExecuteCosmeticEvents
        (:True
          (CallFunction|BindMediaSourceEvents)
          (CallFunction|BindMediaPlayerEvents (Variables|Default|GetStreamingMediaPlayer)))))
    (:then_1
      (bind (_asynctaskproxy _payload) (GameplayEventRouter|StartListeningToEventBackwardCompatible "(TagName=\"EventMessage.Component.Playspace.PlayerEnteredSpatialActor.LegacyMinigame\")" (GameplayEventRouter|GetGlobalEventRouter) "/Script/FortniteGame.PlayerEnterSpatialActorContextLegacyMinigame" 0 "Down")
        (:then
          (Variables|Default|SetPlayerEnteredHandle _asynctaskproxy))
        (:OnEventReceived
          ;;[0.6305 1 0.6947] Reset on Entering a new Volume
          (Networking|CanExecuteCosmeticEvents
            (:True
              (if (Utilities|Equal(Object) (.EnteringPlayerState (Utilities|Struct|BreakPlayerEnterSpatialActorContextLegacyMinigame _payload)) (Variables|/Script/Engine.Controller|GetPlayerState (Game|GetPlayerController 0)))
                (if (CallFunction|IsClientPlayerInSameVolume (.EnteringPlayerState (Utilities|Struct|BreakPlayerEnterSpatialActorContextLegacyMinigame _payload)))
                  (CallFunction|StartStreamingVideo false))))))))))

(event EventEndPlay
  (params (EndPlayReason :type EEndPlayReason%20Enum :default "Destroyed"))
  (bind _playerenteredhandle (Variables|Default|GetPlayerEnteredHandle))
  (Utilities|IsValid _playerenteredhandle
    (:Is%20Valid
      (GameplayEventRouter|StopListeningtoEvent _playerenteredhandle))
    (:Is%20Not%20Valid
      _))
  (Networking|CanExecuteCosmeticEvents
    (:True
      (CallFunction|UnbindMediaPlayerEvents (Variables|/Script/FortniteGame.FortMediaPlayerCtrl|GetMediaPlayer (Variables|Default|GetMediaPlayerCtrl)))
      (CallFunction|UnbindMediaSourceEvents)
      (StreamingMediaLoading|OnStreamingTurnedOff)
      (CallFunction|EnableShareblocking false))))

(event Custom|StartStreamingVideo
  (params (isRetry :type Boolean))
  (Networking|CanExecuteCosmeticEvents
    (:True
      (Utilities|FlowControl|DelayUntilNextTick) ; Just to ensure we only try to start streaming the video once a frame.
      ;[0 1 0.6471 0.5 22] Start streaming video : Check to see if we can actualy stream and if so store time entered zone, set streaming material, assign the sound component to the streaming media player, request the video url for the media source, have the mediaplayer open the media source which will start the streaming.
      (if (CallFunction|IsClientPlayerInSameVolume 0)
        (Class|CreativeBlueprintLibrary|LogCreativeStreamingMessage (Utilities|String|Append "StartStreamingVideo: " (Utilities|String|EnumtoString (Variables|Default|GetClient%20Streaming%20State))))
        (if (not isRetry)
          (Variables|Default|SetRetries 0) ; Reset Retry Attempts
        )
        (Variables|ControllingActor|GetControlling%20Stream%20Actor
          (:then
            (bind _available (StreamingMediaLoading|IsStreamingAvailable))
            (if _available
              (Variables|Default|SetbTimeChanged (!= (Variables|Default|GetLocalStartTime) (Variables|Default|GetServerStartTime)))
              (Variables|Default|SetLocalStartTime (Variables|Default|GetServerStartTime))
              (if (CallFunction|ShouldPlay)
                (Variables|Default|SetLoadURLLoopCount 0)
                (Utilities|LoadURLIfNeeded true
                  (:NotLoaded
                    (Utilities|HasUCPBoth
                      (:False
                        (CallFunction|UpdateStreamState "NewEnumerator6"))))
                  (:Loaded
                    (Utilities|HasUCPBoth
                      (:False
                        (if (Variables|Default|GetbTimeChanged)
                          (if (Utilities|Enum|Equal(Enum) (Variables|Default|GetClient%20Streaming%20State) "NewEnumerator3")
                            (Class|CreativeBlueprintLibrary|LogCreativeStreamingMessage "Video Was Streaming Just Seeking")
                            (StreamingMediaLoading|CurrentSourceHasPartySync
                              (:Yes
                                (StreamingMediaLoading|SetupPartySyncAndSeek)))
                            (else
                              (CallFunction|OnVideoURLSuccess (Media|MediaSource|GetUrl (Variables|Default|GetMediaSource)) 0 "BLURL")))))))))
              (else
                (Class|CreativeBlueprintLibrary|LogCreativeStreamingMessage "Streaming Is Unavailable"))))
          (:else)) ; only start if we have a valid stream object
      ))))

(event Custom|Request%20Streaming%20Update
  (params (Requesting%20Actor :type Actor%20Object%20Reference))
  (Utilities|SwitchHasAuthority
    (:Authority
      (CallFunction|ChooseNewControllingStreamActor)
      (CallFunction|UpdateServerStartTime)
      (CallFunction|UpdateServerIsControllingPlaying))))

(event Custom|OnSourceOpenFailed
  (params (FailedUrl :type String))
  ;[1 0.2818 0] Stop streaming video :  - close the streaming media player - call Stop then check for unbinding events and send analytics before starting new streams
  ;;[1 1 1] Media has failed to load, unbind events.
  (Class|CreativeBlueprintLibrary|LogCreativeStreamingMessage (Utilities|String|Append "SourceOpenFailedFor" FailedUrl)))

(event Custom|Server%20Set%20Stream%20Control
  (params (New%20Stream :type Actor%20Object%20Reference) (Lock%20Stream :type Boolean))
  (Utilities|SwitchHasAuthority
    (:Authority
      (if (Utilities|NotEqual(Object) New%20Stream (Variables|ControllingActor|GetControlling%20Stream%20Actor))
        (CallFunction|SetLockStream Lock%20Stream)
        (Variables|ControllingActor|SetControlling%20Stream%20Actor New%20Stream)
        (CallFunction|UpdateServerStartTime)
        (CallFunction|UpdateServerIsControllingPlaying)))))

(event Custom|OnMediaOpenFailed
  (params (FailedUrl :type String))
  ;[1 0.2818 0] Stop streaming video :  - close the streaming media player - call Stop then check for unbinding events and send analytics before starting new streams
  ;;[1 1 1] Media has failed to load, unbind events.
  (Class|CreativeBlueprintLibrary|LogCreativeStreamingMessage (Utilities|String|Append "OnMediaOpenFailed: " FailedUrl)))

(event Custom|Server%20Request%20Release%20Stream%20Control
  (Utilities|SwitchHasAuthority
    (:Authority
      (Variables|ControllingActor|SetLock%20Stream%20to%20Controlling%20Actor false)
      (CallFunction|ChooseNewControllingStreamActor))))

(event Custom|CheckStreamOnInit
  (params (Requesting%20Actor :type Actor%20Object%20Reference))
  (Utilities|SwitchHasAuthority
    (:Authority
      (CallFunction|ChooseNewControllingStreamActor))))

(event Custom|CheckForNewIsland
  (Utilities|FlowControl|DelayUntilNextTick)
  (if (CallFunction|IsClientPlayerInSameVolume 0)
    _))

(event Custom|SetStartLocked
  ;[1 1 1] Set the bool
  (Variables|Default|SetbStartLocked true))

(event Custom|CheckForSettingsOnlyChange)

(event Custom|OnMediaEnded
  ;[1 1 1] OnMediaEnded
  (Class|CreativeBlueprintLibrary|LogCreativeStreamingMessage "MediaEnded")
  (bind _controlling_stream_actor (Variables|ControllingActor|GetControlling%20Stream%20Actor)
    (:then
      (bind _wantstoloop (VideoStreaming|Source|GetWantstoLoop(Message) _controlling_stream_actor))
      (Utilities|FlowControl|Delay 0.5)
      (if _wantstoloop
        (StreamingMediaLoading|CurrentSourceHasPartySync
          (:Yes
            (Media|MediaPlayer|Seek (Variables|/Script/FortniteGame.FortMediaPlayerCtrl|GetMediaPlayer (Variables|Default|GetMediaPlayerCtrl)) (Math|Timespan|MakeTimespan 0)))
          (:No
            (Media|MediaPlayer|Seek (Variables|Default|GetStreamingMediaPlayer) (Math|Timespan|MakeTimespan 0))))
        (else
          (StreamingMediaLoading|CurrentSourceHasPartySync
            (:Yes
              (Media|MediaPlayer|Close (Variables|/Script/FortniteGame.FortMediaPlayerCtrl|GetMediaPlayer (Variables|Default|GetMediaPlayerCtrl)) :!class /Script/MediaAssets.MediaPlayer))
            (:No
              (Media|MediaPlayer|Close (Variables|Default|GetStreamingMediaPlayer) :!class /Script/MediaAssets.MediaPlayer))))))
    (:else)))

(fn ActorcanPlay
  (params (Actor :type Actor%20Object%20Reference%20%28by%20ref%29))
  (bind (_stream_uid _start_time _priority) (VideoStreaming|Source|GetStreamDetails(Message) Actor))
  (return (>= _priority 0.0)))

(fn AddStreamSource
  (params (Source :type Actor%20Object%20Reference))
  (Utilities|SwitchHasAuthority
    (:Authority
      (Utilities|Set|Add (Variables|Default|GetStreamSources) Source))))

(fn BindMediaPlayerEvents
  (params (TargetPlayer :type Media%20Player%20Object%20Reference))
  (Media|MediaPlayer|BindEventtoOnMediaOpenFailed TargetPlayer (EventDispatchers|CreateEvent :!event OnMediaOpenFailed))
  (Media|MediaPlayer|BindEventtoOnMediaClosed TargetPlayer (EventDispatchers|CreateEvent :!event OnMediaClosed))
  (Media|MediaPlayer|BindEventtoOnMediaOpened TargetPlayer (EventDispatchers|CreateEvent :!event OnMediaReady))
  (Media|MediaPlayer|BindEventtoOnEndReached TargetPlayer (EventDispatchers|CreateEvent :!event OnMediaEnded))
  (Media|MediaPlayer|BindEventtoOnPlaybackResumed TargetPlayer (EventDispatchers|CreateEvent :!event OnMediaResumed)))

(fn BindMediaSourceEvents
  (Stream|BindEventtoOnVideoUrlSuccess (Variables|Default|GetMediaSource) (EventDispatchers|CreateEvent :!event OnVideoURLSuccess))
  (Stream|BindEventtoOnVideoUrlFailed (Variables|Default|GetMediaSource) (EventDispatchers|CreateEvent :!event OnVideoURLFailure)))

(fn CheckforValidLockRequest
  (params (Streaming%20Actor :type Actor%20Object%20Reference))
  (return (Utilities|NotEqual(Object) Streaming%20Actor (Variables|ControllingActor|GetControlling%20Stream%20Actor))))

(fn ChooseNewControllingStreamActor
  ;[1 1 1] Iterates through all of our streaming players and picks the one with the highest priority to be our new controlling actor.
  ; Disabled streaming actors should have a negative priority so they're never picked.
  (Utilities|SwitchHasAuthority
    (:Authority
      (if (not (Variables|ControllingActor|GetLock%20Stream%20to%20Controlling%20Actor))
        (bind _result (Utilities|Set|ToArray (Variables|Default|GetStreamSources)))
        (bind _output_get (LocalVariables|Setlocal%20Player%20Array _result))
        (for _array_element _output_get
          (bind _stream_priority (VideoStreaming|Source|GetStreamPriority(Message) _array_element))
          (if (> _stream_priority (LocalVariables|GetBest%20Priority))
            (LocalVariables|SetBest%20Priority _stream_priority)
            (LocalVariables|SetBest%20Streaming%20Actor _array_element)))
        (bind _best_streaming_actor (LocalVariables|GetBest%20Streaming%20Actor)
          (:then
            (CallFunction|ServerSetStreamControl _best_streaming_actor))
          (:else))))))

(fn CloseStreamingVideo
  (Networking|CanExecuteCosmeticEvents
    (:True
      (Class|CreativeBlueprintLibrary|LogCreativeStreamingMessage "CloseStreamingVideo")
      (Media|MediaPlayer|Close (Variables|Default|GetStreamingMediaPlayer) :!class /Script/MediaAssets.MediaPlayer)
      (FortniteSubtitles|SubtitlesPlayer|Stop (Variables|Default|GetSubtitlesPlayer))
      (StreamingMediaLoading|OnStreamingTurnedOff))))

(fn EnableShareblocking
  (params (Enable :type Boolean))
  (bind _mediasource (Variables|Default|GetMediaSource)
    (:then
      (if Enable
        (if (and (not (Variables|Default|GetShareblockingEnabled)) (Variables|/Script/FortniteGame.FortStreamMediaSource|GetbShareLock _mediasource))
          (Variables|Default|SetShareblockingEnabled true)
          (Player|DisableSharing _mediasource))
        (elif (and (Variables|Default|GetShareblockingEnabled) (Variables|/Script/FortniteGame.FortStreamMediaSource|GetbShareLock _mediasource))
          (Variables|Default|SetShareblockingEnabled false)
          (Player|ReinstateSharing _mediasource))))
    (:else)))

(fn GetActiveStreamDetails
  (bind _controlling_stream_actor (Variables|ControllingActor|GetControlling%20Stream%20Actor)
    (:then
      (bind (_stream_uid _start_time _priority) (VideoStreaming|Source|GetStreamDetails(Message) _controlling_stream_actor))
      (return (>= _priority 0.0) _start_time _controlling_stream_actor))
    (:else)))

(fn GetCurrentStream
  (bind _mediasource (Variables|Default|GetMediaSource))
  (Utilities|IsValid _mediasource
    (:Is%20Valid
      (return (Variables|/Script/FortniteGame.FortStreamMediaSource|GetStreamID _mediasource) (Utilities|String|NotEqualExactly(String) (Variables|/Script/FortniteGame.FortStreamMediaSource|GetStreamID _mediasource) (Variables|Default|GetDefault%20Stream))))
    (:Is%20Not%20Valid
      (return "" false))))

(fn GetIDFromControllingObject
  (bind _controlling_stream_actor (Variables|ControllingActor|GetControlling%20Stream%20Actor)
    (:then
      (bind _videoid (VideoStreaming|Source|GetStreamingVideoID(Message) _controlling_stream_actor))
      (return _videoid))
    (:else
      (return ""))))

(fn GetIsPlayingFromControllingObject
  (bind _controlling_stream_actor (Variables|ControllingActor|GetControlling%20Stream%20Actor)
    (:then
      (bind _isplaying (VideoStreaming|Source|GetStreamIsPlaying(Message) _controlling_stream_actor))
      (return _isplaying))
    (:else
      (return false))))

(fn GetSeekTime
  (return (Math|Timespan|FromSeconds (Math|Float|%25%28Float%29 (- (Fortnite|Sync|Time|GetServerWorldTimeSeconds) (Variables|Default|GetLocalStartTime)) (Math|Float|Max(Float) (Math|Timespan|GetTotalSeconds (Media|MediaPlayer|GetDuration (Variables|Default|GetStreamingMediaPlayer))) 1.0)))))

(fn GetStartTimeFromControllingObject
  (bind _controlling_stream_actor (Variables|ControllingActor|GetControlling%20Stream%20Actor)
    (:then
      (bind _starttime (VideoStreaming|Source|GetStreamStartTime(Message) _controlling_stream_actor))
      (return _starttime))
    (:else
      (return -1.0))))

(fn GetStreamDetails
  (return "" 0.0 0.0))

(fn GetStreamingVideoID
  (return (CallFunction|GetIDfromControllingObject)))

(fn GetStreamIsPlaying
  (return false))

(fn GetStreamPriority
  (return 0.0))

(fn GetStreamStartTime
  (return 0.0))

(fn GetWantsToLoop
  (return false))

(fn IsClientPlayerinSameVolume
  (params (Client :type Player%20State%20Object%20Reference))
  (Networking|CanExecuteCosmeticEvents
    (:True
      (bind _returnvalue (Class|CreativeBlueprintLibrary|IsActorRelevanttoActor (Components|GetOwner) (select (Utilities|IsValid Client :!class /Script/Engine.KismetSystemLibrary) Client (Variables|/Script/Engine.Controller|GetPlayerState (Game|GetPlayerController 0)))))
      (return _returnvalue))
    (:False
      (return false))))

(fn IsStreamingAvailable
  (Utilities|SwitchHasAuthority
    (:Remote
      ;[1 1 1] Check to see if television streaming off
      (bind _returnvalue (Utilities|GetCommandLine))
      (if (Utilities|String|Contains _returnvalue "TelevisionOff")
        (return false)
        (else
          (return true))))))

(fn OnMediaSourceChanged)

(fn OnStreamingTurnedOff
  (for _array_element (Fortnite|GetLocalFortPlayerControllers true)
    (Movie|SendStreamingAnalytics _array_element (Variables|Default|GetStreamingMediaPlayer))))

(fn OnMediaClosed
  (Class|CreativeBlueprintLibrary|LogCreativeStreamingMessage "OnMediaClosed")
  (CallFunction|EnableShareblocking false)
  (bind _controlling_stream_actor (Variables|ControllingActor|GetControlling%20Stream%20Actor)
    (:then
      (bind _wantstoloop (VideoStreaming|Source|GetWantstoLoop(Message) _controlling_stream_actor))
      (if _wantstoloop
        (Variables|/Script/MediaAssets.MediaTexture|SetAutoClear (Variables|Default|GetMediaTextureAutoClear) (Variables|Default|GetMediaTexture)))
      (FortniteSubtitles|SubtitlesPlayer|Stop (Variables|Default|GetSubtitlesPlayer))
      (CallFunction|UpdateStreamState "NewEnumerator2")
      (if (Variables|Default|GetServerIsControllingPlaying)
        (CallFunction|StartStreamingVideo false)))
    (:else
      (CallFunction|UpdateStreamState "NewEnumerator2"))))

(fn OnMediaReady
  (params (OpenedUrl :type String))
  (Class|CreativeBlueprintLibrary|LogCreativeStreamingMessage (Utilities|String|Append "OnMediaReady OpenedURL: " OpenedUrl))
  (if (Utilities|Enum|Equal(Enum) (Variables|Default|GetClient%20Streaming%20State) "NewEnumerator3")
    (CallFunction|EnableShareblocking true)
    (StreamingMediaLoading|PlayMedia)))

(fn OnMediaResumed)

(fn OnRep_ControllingStreamActor
  ;[1 0.5605 0.9895] Rep the assignment of a new controlling stream object
  (Networking|CanExecuteCosmeticEvents
    (:True
      (CallFunction|StartStreamingVideo false)
      (Default|CallOnStreamSourceChanged (Variables|ControllingActor|GetControlling%20Stream%20Actor) (CallFunction|GetIDfromControllingObject)))))

(fn OnRep_ServerIsControllingPlaying
  (Networking|CanExecuteCosmeticEvents
    (:True
      (if (Variables|Default|GetServerIsControllingPlaying)
        (CallFunction|StartStreamingVideo false)
        (else
          (CallFunction|CloseStreamingVideo))))))

(fn OnRep_StartTime
  (CallFunction|StartStreamingVideo false))

(fn OnVideoURLFailure
  (params (URL :type String) (Code :type Integer) (RequestType :type EEpicMediaRequestType%20Enum :default "UNKNOWN"))
  (Class|CreativeBlueprintLibrary|LogCreativeStreamingMessage (Utilities|String|Append "OnVideoURLFailure" URL))
  (Variables|Default|SetbIsURLLoading false)
  (Utilities|LoadURLIfNeeded false
    (:Loaded
      (CallFunction|UpdateStreamState "NewEnumerator5"))))

(fn OnVideoURLSuccess
  (params (URL :type String) (Code :type Integer) (RequestType :type EEpicMediaRequestType%20Enum :default "UNKNOWN"))
  (Class|CreativeBlueprintLibrary|LogCreativeStreamingMessage (Utilities|String|Append "OnVideoURLSuccess" URL))
  (Variables|Default|SetbIsURLLoading false)
  (Utilities|LoadURLIfNeeded false
    (:Loaded
      (Utilities|HasUCPBoth
        (:True
          (CallFunction|UpdateStreamState "NewEnumerator2"))
        (:False
          (Variables|Default|SetLocalStartTime (Variables|Default|GetServerStartTime))
          (StreamingMediaLoading|SetupPartySyncAndSeek)
          (bind _returnvalue (StreamingMediaLoading|OpenMediaSource))
          (Class|CreativeBlueprintLibrary|LogCreativeStreamingMessage (Utilities|String|Append "OpenMediaSource Result:" (Utilities|String|ToString(Boolean) _returnvalue)))
          (if _returnvalue
            (CallFunction|UpdateStreamState "NewEnumerator3")
            (else
              (CallFunction|UpdateStreamState "NewEnumerator5"))))))))

(fn OpenMediaSource
  (bind _mediatexture (Variables|Default|GetMediaTexture))
  (Class|CreativeBlueprintLibrary|LogCreativeStreamingMessage (Utilities|String|Append "Open Media Source" (Variables|/Script/FortniteGame.FortStreamMediaSource|GetStreamID (Variables|Default|GetMediaSource))))
  (bind _controlling_stream_actor (Variables|ControllingActor|GetControlling%20Stream%20Actor)
    (:then
      (bind _wantstoloop (VideoStreaming|Source|GetWantstoLoop(Message) _controlling_stream_actor))
      (if _wantstoloop
        (Variables|Default|SetMediaTextureAutoClear (Variables|/Script/MediaAssets.MediaTexture|GetAutoClear _mediatexture))
        (Variables|/Script/MediaAssets.MediaTexture|SetAutoClear false _mediatexture))
      (StreamingMediaLoading|CurrentSourceHasPartySync
        (:Yes
          (bind _returnvalue (Media|MediaPlayer|OpenSourcewithOptions (Variables|Default|GetMediaPlayerCtrl) (Variables|Default|GetMediaSource) (Utilities|Struct|MakeMediaPlayerOptions "(Audio=0,Caption=-1,Metadata=-1,Script=-1,Subtitle=-1,Text=-1,Video=0)" "(Video=\"\",Audio=\"\",Subtitle=\"\",Caption=\"\")" "UseTrackOptionIndices" "+00000000.00:00:00.000000000" "RelativeToStartTime" "Enabled" "Disabled") :!class /Script/FortniteGame.FortMediaPlayerCtrl))
          (FortniteSubtitles|SubtitlesPlayer|BindtoMediaPlayer (Variables|Default|GetSubtitlesPlayer) (Variables|/Script/FortniteGame.FortMediaPlayerCtrl|GetMediaPlayer (Variables|Default|GetMediaPlayerCtrl)))
          (if _returnvalue
            (FortniteSubtitles|SubtitlesPlayer|SetSubtitles (Variables|Default|GetSubtitlesPlayer) (Variables|Default|GetSubtitlesOverlay))
            (else
              _))
          (return _returnvalue))
        (:No
          (bind _returnvalue_2 (Media|MediaPlayer|OpenSourcewithOptions (Variables|Default|GetStreamingMediaPlayer) (Variables|Default|GetMediaSource) (Utilities|Struct|MakeMediaPlayerOptions "(Audio=0,Caption=-1,Metadata=-1,Script=-1,Subtitle=-1,Text=-1,Video=0)" "(Video=\"\",Audio=\"\",Subtitle=\"\",Caption=\"\")" "UseTrackOptionIndices" (SeekingandSynching|GetSeekTime) "RelativeToStartTime" "Enabled" "Disabled") :!class /Script/MediaAssets.MediaPlayer))
          (FortniteSubtitles|SubtitlesPlayer|BindtoMediaPlayer (Variables|Default|GetSubtitlesPlayer) (Variables|Default|GetStreamingMediaPlayer))
          (if _returnvalue_2
            (FortniteSubtitles|SubtitlesPlayer|SetSubtitles (Variables|Default|GetSubtitlesPlayer) (Variables|Default|GetSubtitlesOverlay))
            (else
              _))
          (return _returnvalue_2))))
    (:else)))

(fn PlayMedia
  (bind _mediasource (Variables|Default|GetMediaSource)
    (:then
      (if (and (Variables|/Script/FortniteGame.FortStreamMediaSource|GetbPartySync _mediasource) (ScalableFloat|GetValueasBool (Variables|Default|GetUseMediaPlayerController)))
        (Media|MediaPlayer|Play (Variables|Default|GetMediaPlayerCtrl) :!class /Script/FortniteGame.FortMediaPlayerCtrl)
        (else
          (Media|MediaPlayer|Play (Variables|Default|GetStreamingMediaPlayer) :!class /Script/MediaAssets.MediaPlayer)))
      (FortniteSubtitles|SubtitlesPlayer|Play (Variables|Default|GetSubtitlesPlayer)))
    (:else)))

(fn RegisterSelfwithManager
  (bind _local_register_ref (LocalVariables|Getlocal%20Register%20Ref))
  ;[1 1 1] Registers Self with Manager. In two steps as the Create Object Register function auto-registers the context actor (the parent), while we want to register this component.
  (CreativeObjectRegistry|GetOrCreateObjectRegister (Components|GetOwner) "(TagName=\"Creative.Register.Streaming.Manager\")" false 0 _local_register_ref)
  (Class|CreativeIndexedObjectRegisterComp|RegisterObject _local_register_ref self 0))

(fn RemoveStreamSource
  (params (StreamPlayer :type Actor%20Object%20Reference))
  (Utilities|SwitchHasAuthority
    (:Authority
      (Utilities|Set|Remove (Variables|Default|GetStreamSources) StreamPlayer)
      (if (Utilities|Equal(Object) (Variables|ControllingActor|GetControlling%20Stream%20Actor) (Variables|GetStreamPlayer))
        (CallFunction|ChooseNewControllingStreamActor)))))

(fn SetLockStream
  (params (Lock :type Boolean))
  ;[1 1 1] Setting in a function with a check against current to avoid a Notify infinite loop on the server
  (if (Math|Boolean|NotEqual(Boolean) Lock (Variables|ControllingActor|GetLock%20Stream%20to%20Controlling%20Actor))
    (Variables|ControllingActor|SetLock%20Stream%20to%20Controlling%20Actor Lock)))

(fn SetupPartySyncandSeek
  (bind _mediaplayerctrl (Variables|Default|GetMediaPlayerCtrl))
  (Networking|CanExecuteCosmeticEvents
    (:True
      (StreamingMediaLoading|CurrentSourceHasPartySync
        (:Yes
          (Variables|Default|GetMediaPlayerCtrl
            (:then
              _)
            (:else
              (bind _returnvalue (Game|ConstructObjectfromClass "/Script/FortniteGame.FortMediaPlayerCtrl" (Components|GetOwner)))
              (Variables|Default|SetMediaPlayerCtrl _returnvalue)))
          (Variables|/Script/FortniteGame.FortMediaPlayerCtrl|SetMediaPlayer (Variables|Default|GetStreamingMediaPlayer) _mediaplayerctrl)
          (Variables|/Script/FortniteGame.FortMediaPlayerCtrl|SetMediaSource (Variables|Default|GetMediaSource) _mediaplayerctrl)
          (Media|MediaPlayer|SetPlayerStartTime _mediaplayerctrl (* (Variables|Default|GetLocalStartTime) -1.0)))
        (:No
          ;[1 0.5253 0 0.2] If we have a valid media player controller, tell it to stop controlling things (we don't currently want to be using it, our current video doesn't support it)
          (bind _mediaplayerctrl_2 (Variables|Default|GetMediaPlayerCtrl)
            (:then
              (Variables|/Script/FortniteGame.FortMediaPlayerCtrl|SetMediaPlayer 0 _mediaplayerctrl_2)
              (Variables|/Script/FortniteGame.FortMediaPlayerCtrl|SetMediaSource 0 _mediaplayerctrl_2))
            (:else)))))))

(fn ShouldPlay
  (return (and (Variables|Default|GetServerIsControllingPlaying) (Utilities|IsValid (Variables|ControllingActor|GetControlling%20Stream%20Actor) :!class /Script/Engine.KismetSystemLibrary))))

(fn StreamUpdated
  (params (Current%20VUID :type String) (Start%20Time :type Float%20%28double-precision%29))
  (return false))

(fn UnbindMediaPlayerEvents
  (params (Target%20Player :type Media%20Player%20Object%20Reference))
  (Media|MediaPlayer|UnbindEventfromOnMediaOpenFailed Target%20Player (EventDispatchers|CreateEvent :!event OnMediaOpenFailed))
  (Media|MediaPlayer|UnbindEventfromOnMediaClosed Target%20Player (EventDispatchers|CreateEvent :!event OnMediaClosed))
  (Media|MediaPlayer|UnbindEventfromOnMediaOpened Target%20Player (EventDispatchers|CreateEvent :!event OnMediaReady))
  (Media|MediaPlayer|UnbindEventfromOnEndReached Target%20Player (EventDispatchers|CreateEvent :!event OnMediaEnded))
  (Media|MediaPlayer|UnbindEventfromOnPlaybackResumed Target%20Player (EventDispatchers|CreateEvent :!event OnMediaResumed)))

(fn UnbindMediaSourceEvents
  (Stream|BindEventtoOnVideoUrlSuccess (Variables|Default|GetMediaSource) (EventDispatchers|CreateEvent :!event OnVideoURLSuccess))
  (Stream|BindEventtoOnVideoUrlFailed (Variables|Default|GetMediaSource) (EventDispatchers|CreateEvent :!event OnVideoURLFailure)))

(fn UpdateStreamState
  (params (New%20Stream%20State :type ECreativeStreamPlayState%20Enum :default "NewEnumerator0"))
  (if (Utilities|Enum|NotEqual(Enum) New%20Stream%20State (Variables|Default|GetClient%20Streaming%20State))
    (bind _output_get (Variables|Default|SetClient%20Streaming%20State New%20Stream%20State))
    (Default|CallOnStreamStateChanged _output_get)
    (Class|CreativeBlueprintLibrary|LogCreativeStreamingMessage (Utilities|String|Append "Client Stream State Changed to: " (Utilities|String|EnumtoString _output_get)))
    (else
      (Class|CreativeBlueprintLibrary|LogCreativeStreamingMessage "Client Stream State Remained Unchanged"))))

(fn UpdateServerIsControllingPlaying
  (bind _isplaying (CallFunction|GetIsPlayingfromControllingObject))
  (if (Math|Boolean|NotEqual(Boolean) _isplaying (Variables|Default|GetServerIsControllingPlaying))
    (Variables|Default|SetServerIsControllingPlaying _isplaying)))

(fn UpdateServerStartTime
  (bind _starttime (CallFunction|GetStartTimefromControllingObject))
  (if (!= _starttime (Variables|Default|GetServerStartTime))
    (Variables|Default|SetServerStartTime _starttime)
    (VideoStreaming|Source|CheckforTransmit(Message) (Variables|ControllingActor|GetControlling%20Stream%20Actor) (CallFunction|GetIDfromControllingObject))))
