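; Passive ability that decides when Creative TV (Llamatron) streams show picture-in-picture for the owning player, on cosmetic-capable machines.
; An object register feeds an active TV list, tested round-robin with a delay between checks.
; Each check caches a per-TV state from a throttled async visibility trace, scale-versus-distance readability and facing/aim dot products.
; Combined states toggle the live-stream HUD and PiP audio, with any in-view screen suppressing PiP.
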(event Custom|Test%20TV%20Can%20PIP
  (params (Test%20TV :type BP%20Creative%20TV%20Stream%20Object%20Reference))
  (bind _test_tv (Variables|Default|GetTest%20TV))
  ;[0 0.7912 1 0.5] Testing if a TV wants to show Picture in Picture. This includes an Asyncronous Line Trace, followed by a Dot Product test
  (Variables|Default|SetTest%20TV Test%20TV)
  (if (PIP|SkipPipCheck Test%20TV :!class /Game/Creative/Devices/TV_Screen/BP_Creative_TV_Stream.BP_Creative_TV_Stream_C)
    (PIP|SetCachedPIPState (Variables|Default|GetTest%20TV) true :!class /Game/Creative/Devices/TV_Screen/BP_Creative_TV_Stream.BP_Creative_TV_Stream_C)
    (CallFunction|PushTestTVToArray) ; We've handled this TV, place it on the end of the test list
    (CallFunction|OnTestCompleted)
    (else
      (bind _returnvalue (Utility|TraceQueue|GetAsyncTraceQueue _test_tv))
      (Class|AsyncTraceQueue|ThrottledAsyncTraceSourceTarget _returnvalue true 0 (0.0 0.0 0.0) _test_tv (0.0 0.0 0.0) (Utilities|Array|MakeArray _test_tv (Variables|Default|GetAbilityOwner)) "ECC_Visibility" (delegate Custom|Line%20of%20Sight%20Check)))))

(event Ability|EventActivateAbility
  (|CallParentFunction)
  (Networking|CanExecuteCosmeticEvents
    (:True
      (bind _has_active_tvs (CallFunction|UpdateActiveTVsFromRegister))
      (if _has_active_tvs
        (CallFunction|StartTestingFirstTVInList)))))

(event Ability|EventFailedToActivatePassiveAbility
  (params (ActorInfo :type Gameplay%20Ability%20Actor%20Info%20Structure))
  (|CallParentFunction)
  (Ability|EndAbility))

(event Custom|Start%20Testing%20First%20TV%20in%20List
  ;[0 0.7912 1 0.5] Testing a TV in the list. This will pop the first TV off the list after a delay and run some tests to see if it wants to show PIP.
  (Utilities|FlowControl|Delay (Variables|Default|GetDelay%20between%20Tests)) ; Delay between tests
  (bind (_tv_array_empty _success) (CallFunction|PopTestTVFromArray))
  (if _tv_array_empty
    (Ability|CancelAbility)
    (elif _success
      (CallFunction|TestTVCanPIP (Variables|Default|GetTest%20TV)))))

(event Custom|Line%20of%20Sight%20Check
  (params (Succeeded :type Boolean) (ActorHit :type Actor%20Object%20Reference) (HitResult :type Hit%20Result%20Structure))
  ;[0 0.7912 1 0.5] Testing if a TV wants to show Picture in Picture. This includes an Asyncronous Line Trace, followed by a Dot Product test
  (Utilities|FlowControl|Sequence
    (:then_0
      (Utilities|IsValid ActorHit
        (:Is%20Valid
          (bind _test_tv (Variables|Default|GetTest%20TV)
            (:then
              (PIP|SetCachedPIPState _test_tv true :!class /Game/Creative/Devices/TV_Screen/BP_Creative_TV_Stream.BP_Creative_TV_Stream_C) ; Trace hit something, so we can't see the TV
            )
            (:else)))
        (:Is%20Not%20Valid
          (bind _test_tv_2 (Variables|Default|GetTest%20TV)
            (:then
              (bind (_show _distance_to_actor) (CallFunction|TestDistancetoActor _test_tv_2))
              (if _show
                (PIP|SetCachedPIPState _test_tv_2 true :!class /Game/Creative/Devices/TV_Screen/BP_Creative_TV_Stream.BP_Creative_TV_Stream_C)
                (else
                  (bind _should_show (CallFunction|TestDotProduct _test_tv_2 _distance_to_actor))
                  (PIP|SetCachedPIPState _test_tv_2 _should_show :!class /Game/Creative/Devices/TV_Screen/BP_Creative_TV_Stream.BP_Creative_TV_Stream_C))))
            (:else)))))
    (:then_1
      (CallFunction|PushTestTVToArray) ; We've handled this TV, place it on the end of the test list
      (CallFunction|OnTestCompleted))))

(event Custom|On%20Test%20Completed
  ;[1 0.6433 0 0.25] Test has been completed. Test all of our PIP States for active llamatrons. If any of them don't want to show PIP, don't show it.
  (bind _showpip (CallFunction|CheckAllPIPStates))
  (CallFunction|SetPiPStatus _showpip)
  ;;[0.224 0.3571 1 0] Loop back around to test the next TV.
  (CallFunction|StartTestingFirstTVInList) ; Start our Test again
)

(event Ability|EventOnEndAbility
  (params (bWasCancelled :type Boolean))
  ;[1 0.1588 0 0.2] Always disable PIP Status when the Ability gets removed.
  (|CallParentFunction bWasCancelled)
  (Networking|CanExecuteCosmeticEvents
    (:True
      (CallFunction|SetPiPStatus false true))))

(fn CheckAllPIPStates
  (if (Math|Conversions|ToBoolean(Integer) (Utilities|Array|Length (Variables|Default|GetActiveTVs)))
    (for _array_element (Variables|Default|GetActiveTVs)
      (Utilities|IsValid _array_element
        (:Is%20Valid
          (if (PIP|GetPIPValuesforScreen _array_element :!class /Game/Creative/Devices/TV_Screen/BP_Creative_TV_Stream.BP_Creative_TV_Stream_C)
            (if (.Screen%20Not%20in%20View (PIP|GetPIPValuesforScreen _array_element :!class /Game/Creative/Devices/TV_Screen/BP_Creative_TV_Stream.BP_Creative_TV_Stream_C))
              (LocalVariables|SetWantsToShowPIP true)
              (else
                (return false)))))))
    (else
      _))
  (return (LocalVariables|GetWantsToShowPIP)))

(fn OnObjectAdded
  (params (Object%20Added :type Object%20Reference) (Index :type Integer))
  (bind _activetvs (Variables|Default|GetActiveTVs))
  (bind _asbp_creative_tv_stream (Utilities|Casting|CastToBP_Creative_TV_Stream Object%20Added))
  (Utilities|Array|AddUnique _activetvs _asbp_creative_tv_stream)
  (Variables|Default|SetForce%20Next%20PIP%20Check true)
  (if (== (Utilities|Array|Length _activetvs) 1)
    (CallFunction|StartTestingFirstTVInList) ; If the active TV length is now 1, it was previously 0.\nStart testing.
  ))

(fn OnObjectRemoved
  (params (Object%20Removed :type Object%20Reference) (Index :type Integer))
  (bind _asbp_creative_tv_stream (Utilities|Casting|CastToBP_Creative_TV_Stream Object%20Removed))
  (Utilities|Array|RemoveItem (Variables|Default|GetActiveTVs) _asbp_creative_tv_stream)
  (PIP|SetPIPAudio _asbp_creative_tv_stream false :!class /Game/Creative/Devices/TV_Screen/BP_Creative_TV_Stream.BP_Creative_TV_Stream_C)
  (Variables|Default|SetForce%20Next%20PIP%20Check true))

(fn PiPEnabled
  (bind _returnvalue (Settings|GetGameUserSettings))
  (bind _asfort_game_user_settings (Utilities|Casting|CastToFortGameUserSettings _returnvalue)
    (:then
      (return (Utilities|Enum|Equal(Enum) (Variables|/Script/FortniteGame.FortGameUserSettings|GetShowLiveStreamPictureInPictureInMatchV2 _asfort_game_user_settings) "Show")))
    (:CastFailed
      (return false))))

(fn PopTestTVfromArray
  (bind _activetvs (Variables|Default|GetActiveTVs))
  (bind _output (Utilities|Array|Get(acopy) _activetvs))
  (if (Utilities|Array|IsValidIndex _activetvs)
    (Utilities|IsValid _output
      (:Is%20Valid
        _)
      (:Is%20Not%20Valid
        (Utilities|Array|RemoveIndex (Variables|Default|GetActiveTVs))))
    (Variables|Default|SetTest%20TV _output)
    (Utilities|Array|RemoveIndex (Variables|Default|GetActiveTVs))
    (return false true))
  (return true false))

(fn PushTestTVtoArray
  (Utilities|Array|AddUnique (Variables|Default|GetActiveTVs) (Variables|Default|GetTest%20TV))
  (Variables|Default|SetTest%20TV 0))

(fn ResetAbilityOwner
  (Variables|Default|GetAbilityOwner
    (:then)
    (:else
      (bind _fortpawn (CallFunction|SetupPawnAbility))
      (Variables|Default|SetAbilityOwner _fortpawn))))

(fn SetPiPStatus
  (params (Show :type Boolean) (Force :type Boolean))
  (bind _show (Variables|GetShow))
  (if (Math|Boolean|ORBoolean (Variables|Default|GetForce%20Next%20PIP%20Check) (Math|Boolean|NotEqual(Boolean) Show (Variables|Default|GetPIP%20State)) (Variables|GetForce))
    (Variables|Default|SetPIP%20State _show)
    (Variables|/Script/FortniteGame.FortRuntimeOptions|SetbEnableLiveStreamInMatch _show (Options|GetRuntimeOptions))
    (Variables|Default|SetForce%20Next%20PIP%20Check false)
    (Class|LiveStreamHUDElement|SetMirrorScreen (Math|Integer|SelectInt 5 -1 _show))
    (for _array_element (Variables|Default|GetActiveTVs)
      (PIP|SetPIPAudio _array_element (and (CallFunction|PiPEnabled) (Variables|GetShow)) :!class /Game/Creative/Devices/TV_Screen/BP_Creative_TV_Stream.BP_Creative_TV_Stream_C))))

(fn TestDotProduct
  (params (TV :type BP%20Creative%20TV%20Stream%20Object%20Reference) (Distance%20to%20Actor :type Float%20%28double-precision%29))
  (bind (_visible_dot_product_min_max_min_scale_x _visible_dot_product_min_max_min_scale_y) (Math|Vector2D|BreakVector2D (Variables|GetVisible%20Dot%20Product%20Min/Max%20Min%20Scale_X)))
  (bind (_dot_distance_min_max_x _dot_distance_min_max_y) (Math|Vector2D|BreakVector2D (Variables|GetDot%20Distance%20Min/Max_X)))
  (LocalVariables|SetDot (Math|Float|MapRangeClamped Distance%20to%20Actor _dot_distance_min_max_x _dot_distance_min_max_y _visible_dot_product_min_max_min_scale_x _visible_dot_product_min_max_min_scale_y))
  (Variables|Default|GetAbilityOwner
    (:then
      _)
    (:else
      ;[1 0.4225 0.4923 0.8] We lost our reference to the owner, try to reacquire it
      (CallFunction|ResetAbilityOwner)
      (Variables|Default|GetAbilityOwner
        (:then
          _)
        (:else
          (return false)))))
  (if (> (Math|Vector|DotProduct (Math|Vector|GetForwardVector (TargetSelection|GetDirectionActortoActor (Variables|Default|GetAbilityOwner) (Variables|GetTV))) (Transformation|GetActorRightVector (Variables|GetTV))) 0.0)
    (return true)
    (else
      (return (< (Math|Vector|DotProduct (Math|Vector|GetForwardVector (TargetSelection|GetDirectionActortoActor (Variables|Default|GetAbilityOwner) (Variables|GetTV))) (Math|Vector|GetForwardVector (Pawn|GetBaseAimRotation (Variables|Default|GetAbilityOwner)))) (LocalVariables|GetDot))))))

(fn TestAnyPiP)

(fn TestDistanceToActor
  (params (TV :type Actor%20Object%20Reference))
  (bind _returnvalue_x (Math|Vector|BreakVector (Transformation|GetActorScale3D (Variables|GetTV))))
  (bind (_pip_scale_min_max_x _pip_scale_min_max_y) (Math|Vector2D|BreakVector2D (Variables|GetPIP%20Scale%20Min/Max_X)))
  (bind (_pip_distance_min_max_x _pip_distance_min_max_y) (Math|Vector2D|BreakVector2D (Variables|GetPIP%20Distance%20Min/Max_X)))
  ;[0.1455 1 0.8701 0.25] Tests if the Llamatron is far enough away to be unreadable. If it is, we want to show PIP. Called from GAT_Passive_CreativeTVPiP when it is this TV's time to check
  (bind _abilityowner (Variables|Default|GetAbilityOwner)
    (:then
      _)
    (:else
      (CallFunction|ResetAbilityOwner)
      (Variables|Default|GetAbilityOwner
        (:then
          _)
        (:else
          (return false 0.0)))))
  (bind _output_get (LocalVariables|SetDistance (Class|AbilityFunctions|GetDistanceBetweenActors TV _abilityowner)))
  (return (> (Math|Float|MapRangeClamped _output_get _pip_distance_min_max_x _pip_distance_min_max_y _pip_scale_min_max_x _pip_scale_min_max_y) (Math|Float|Absolute(Float) _returnvalue_x)) _output_get))

(fn UpdateActiveTVsfromRegister
  (bind _abilityowner (Variables|Default|GetAbilityOwner)
    (:then
      (CreativeObjectRegistry|GetOrCreateObjectRegister _abilityowner "(TagName=\"Creative.Register.Llamatron.All\")" false 0 (Variables|Default|GetTVRegister))
      (bind (_bisvalid _register) (Class|FortCreativeIndexedObjectRegisterComponent|GetRegisteratIndex (Variables|Default|GetTVRegister) 0))
      (bind _register_objects (Utilities|Struct|BreakFortObjectArray _register))
      (if _bisvalid
        (for _array_element _register_objects
          (bind _asbp_creative_tv_stream (Utilities|Casting|CastToBP_Creative_TV_Stream _array_element))
          (Utilities|Array|Add (Variables|Default|GetActiveTVs) _asbp_creative_tv_stream))
        (Default|BindEventtoOnObjectAdded (Variables|Default|GetTVRegister) (EventDispatchers|CreateEvent :!event On%20Object%20Added))
        (Default|BindEventtoOnObjectRemoved (Variables|Default|GetTVRegister) (EventDispatchers|CreateEvent :!event On%20Object%20Removed))
        (return (Math|Conversions|ToBoolean(Integer) (Utilities|Array|Length (Variables|Default|GetActiveTVs))))))
    (:else)))
